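# Stereo one-shot MetaSound for SulfurDean's reactive cosmetic, playing randomly chosen chuckle voice clips on the backbling sound class.
# Play triggers the shared Simple Base cosmetic sound after a short random delay.
# When a clip finishes, a probability filter either retriggers it, lowering the repeat chance each time, or ends the source, allowing decaying laugh chains.
# A second probability-filter branch has no trigger input and appears unused.
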MetaSoundSource /BRCosmetics/Sounds/Characters/SulfurDean/MSS_Reactive_SulfurDean_Preset
inputs [
 Chuckle {type=WaveAsset:Array}
 PitchModulationRange.Max {type=Float}
 PitchModulationRange.Min {type=Float}
 StartDelay.Max {type=Time}
 StartDelay.Min {type=Time}
 TemplatePitchMultiplier {type=Float}
 TemplateVolumeMultiplier {type=Float}
 UE.Source.OnPlay {type=Trigger}
 VolumeModulationRange.Max {type=Float}
 VolumeModulationRange.Min {type=Float}
]
interfaces [UE.OutputFormat.Stereo:1.0, UE.Source.OneShot:1.0, UE.Source:1.1]
outputs [
 "UE.OutputFormat.Stereo.Audio:0" {type=Audio}
 "UE.OutputFormat.Stereo.Audio:1" {type=Audio}
 UE.Source.OneShot.OnFinished {type=Trigger}
]
pages [Default {
  input_defaults {
   PitchModulationRange.Max=1
   PitchModulationRange.Min=1
   StartDelay.Max=0
   StartDelay.Min=0
   TemplatePitchMultiplier=1
   TemplateVolumeMultiplier=1
   UE.Source.OnPlay=null
   VolumeModulationRange.Max=1
   VolumeModulationRange.Min=1
   Chuckle [
    &/VO_SulfurDean_LaughsChuckles_03
    &/VO_SulfurDean_LaughsChuckles_04
    &/VO_SulfurDean_LaughsChuckles_05
    &/VO_SulfurDean_LaughsChuckles_06
    &/VO_SulfurDean_LaughsChuckles_07
    &/VO_SulfurDean_LaughsChuckles_08
    &/VO_SulfurDean_LaughsChuckles_10
    &/VO_SulfurDean_LaughsChuckles_11
   ]
  }
  nodes [
   n1 = InitVariable.Trigger()
   n2 = VariableDeferredAccessor.Float(Variable=InitVariable.Float().Variable)
   n3 = VariableDeferredAccessor.Trigger(Variable=InitVariable.Trigger().Variable)
   TriggerFilter_1 = UE.Trigger Filter(Probability=VariableDeferredAccessor.Float(Variable=@n2.Variable).Value, Trigger=@n3.Value)
   n4 = Value.Value:Float(
    Target Value=UE.Subtract:Float(PrimaryOperand=@n2.Value) {also {defaults {AdditionalOperands=0.25}}}.Out
    Set=@TriggerFilter_1.Heads
    Init Value=@n2.Value
   )
   TriggerAnyTriggerAny2_2 = "TriggerAny.Trigger Any (2)"(In 0=^UE.Source.OnPlay, In 1=@n4.On Set)
   n5 = UE.RandomTime(Next=@TriggerAnyTriggerAny2_2.Out) {also {defaults {Max=0.15}}}
   TriggerDelay_0 = UE.Trigger Delay(In=@n5.On Next, Delay Time=@n5.Value)
   n6 = /FortMetasound/Cosmetics/MSS_Cosmetic_Simple_Base(
    Waves=^Chuckle
    PitchModulationRange.Max=^PitchModulationRange.Max
    PitchModulationRange.Min=^PitchModulationRange.Min
    StartDelay.Max=^StartDelay.Max
    StartDelay.Min=^StartDelay.Min
    TemplatePitchMultiplier=^TemplatePitchMultiplier
    TemplateVolumeMultiplier=^TemplateVolumeMultiplier
    VolumeModulationRange.Max=^VolumeModulationRange.Max
    VolumeModulationRange.Min=^VolumeModulationRange.Min
    UE.Source.OnPlay=VariableAccessor.Trigger(Variable=VariableMutator.Trigger(Variable=@n1.Variable, Value=@TriggerDelay_0.Out).Variable).Value
   )
   AudioMixerAudioMixerStereo2_0 = "AudioMixer.Audio Mixer (Stereo, 2)"(In 0 L=@n6."UE.OutputFormat.Stereo.Audio:0", In 0 R=@n6."UE.OutputFormat.Stereo.Audio:1")
   InitVariable.WaveAsset:Array()
   VariableMutator.Float(Value=@n4.Output Value, Variable=@n2.Variable)
   n7 = VariableDeferredAccessor.Float(Variable=InitVariable.Float().Variable)
   TriggerFilter_0 = UE.Trigger Filter(Probability=VariableDeferredAccessor.Float(Variable=@n7.Variable).Value)
   TriggerAnyTriggerAny2_0 = "TriggerAny.Trigger Any (2)"(In 0=@TriggerFilter_0.Heads, In 1=@TriggerFilter_0.Tails)
   VariableMutator.Float(
    Value=Value.Value:Float(Set=@TriggerAnyTriggerAny2_0.Out, Init Value=@n7.Value) {also {defaults {Target Value=1}}}.Output Value
    Variable=@n7.Variable
   )
   VariableMutator.Trigger(Variable=InitVariable.Trigger().Variable, Value=@TriggerFilter_0.Tails)
   TriggerAnyTriggerAny2_1 = "TriggerAny.Trigger Any (2)"(In 0=@n6.UE.Source.OneShot.OnFinished)
   VariableMutator.Trigger(Value=@TriggerAnyTriggerAny2_1.Out, Variable=@n3.Variable)
  ]
  links [
   {dst="@UE.OutputFormat.Stereo.Audio:0", src=AudioMixerAudioMixerStereo2_0.Out L}
   {dst="@UE.OutputFormat.Stereo.Audio:1", src=AudioMixerAudioMixerStereo2_0.Out R}
   {dst="@UE.Source.OneShot.OnFinished", src=TriggerFilter_1.Tails}
  ]
 }]
properties {
 assetImportData=&self:AssetImportData
 attenuationSettings=&Fort_Pets_Close_Athena_Att
 audioPropertiesSheet=/BRCosmetics/Sounds/Settings/PropertySheets/Reactive/APS_Cosmetic_Reactive_Skin
 maxDistance=1800.0001
 numChannels=2
 outputFormat=Stereo
 sampleRate=48000
 soundClassObject=/Game/Sounds/FortSoundClassesAndModes/Fort_Backblings
 volume=1.5
 concurrencySet [&/CS_SulferDean, &/CS_SulferDean_Global]
}
